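# Looping ambient hum bed for the exterior car wash prop, routed to the indoor Athena ambient sound class.
# A mixer layers three looping car wash hum waves at different input volumes.
# The Hum_02 layer is the quietest and runs through a modulator that fixes it at a slightly lowered pitch, with no randomization.
# Spatialization borrows the close water spray loop attenuation asset.
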SoundCue /Game/Sounds/Fort_Prop_Sounds/Exterior/CarWash/SC_CarWash_AbmientNoise_Loop
root=Mixer_0
nodes [
 {id=Looping_0, children [WavePlayer_1]}
 {id=Looping_1, children [WavePlayer_2]}
 {id=Looping_2, children [WavePlayer_0]}
 {id=Mixer_0, children [Modulator_0,Looping_1,Looping_2], properties {inputVolume [0.3,1,0.8]}}
 {id=Modulator_0, children [Looping_0], properties {pitchMax=0.9, pitchMin=0.9, volumeMax=1, volumeMin=1}}
 {id=WavePlayer_0, properties {bLooping=true, soundWaveAssetPtr=&/CarWash_Ambiance_Hum_04}}
 {id=WavePlayer_1, properties {bLooping=true, soundWaveAssetPtr=&/CarWash_Ambiance_Hum_02}}
 {id=WavePlayer_2, properties {bLooping=true, soundWaveAssetPtr=&/CarWash_Ambiance_Hum_03}}
]
properties {
 attenuationSettings=&/ATT_CarWash_WaterSpray_Loop_Close
 maxDistance=3300
 soundClassObject=&Fort_Ambient_Indoors_Athena
 volumeMultiplier=0.85
}
