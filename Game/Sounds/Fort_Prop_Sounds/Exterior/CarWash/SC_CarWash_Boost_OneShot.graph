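# One-shot sound cue for a car wash boost, routed to the exterior vehicle boost sound class with golf cart boost attenuation.
# A quality switch gives its first two levels a three-layer mix: randomized player-launch and blowhole-burst variants, plus a quieter, delayed cube-surfing boost tail, each through a modulator.
# The last two levels use a cheaper mix of one fixed launch and one burst wave, without randomization, modulation or tail.
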SoundCue /Game/Sounds/Fort_Prop_Sounds/Exterior/CarWash/SC_CarWash_Boost_OneShot
root=QualityLevel_0
nodes [
 {id=Delay_0, children [WavePlayer_11], properties {delayMax=0.3, delayMin=0.3}}
 {id=Mixer_0, children [Modulator_0,Modulator_1,Modulator_2], properties {inputVolume [1,1,0.5]}}
 {id=Mixer_1, children [WavePlayer_4,WavePlayer_5,null], properties {inputVolume [0.8,1,1]}}
 {id=Modulator_0, children [Random_0]}
 {id=Modulator_1, children [Random_1]}
 {id=Modulator_2, children [Delay_0]}
 {id=QualityLevel_0, children [Mixer_0,Mixer_0,Mixer_1,Mixer_1]}
 {id=Random_0, children [WavePlayer_0,WavePlayer_1], properties {weights [1,1]}}
 {id=Random_1, children [WavePlayer_2,WavePlayer_3], properties {weights [1,1]}}
 {id=WavePlayer_0, properties {soundWaveAssetPtr=&/BasilOkra_PlayerLaunched_0S_01}}
 {id=WavePlayer_1, properties {soundWaveAssetPtr=&/BasilOkra_PlayerLaunched_0S_02}}
 {id=WavePlayer_11, properties {soundWaveAssetPtr=&/Ability_CubeSurfing_Boost_Start_01_TailOnly}}
 {id=WavePlayer_2, properties {soundWaveAssetPtr=&/Buttercake_Blowhole_Launch_Blowhole_Burst_02}}
 {id=WavePlayer_3, properties {soundWaveAssetPtr=&/Buttercake_Blowhole_Launch_Blowhole_Burst_01}}
 {id=WavePlayer_4, properties {soundWaveAssetPtr=&BasilOkra_PlayerLaunched_0S_01}}
 {id=WavePlayer_5, properties {soundWaveAssetPtr=&Buttercake_Blowhole_Launch_Blowhole_Burst_01}}
]
properties {
 attenuationSettings=/Game/Sounds/Attenuation/ATT_Fort_GolfCart_Movement_Boost
 maxDistance=12000
 soundClassObject=/Game/Sounds/FortSoundClassesAndModes/Fort_Vehicle_Exterior_Boost
}
