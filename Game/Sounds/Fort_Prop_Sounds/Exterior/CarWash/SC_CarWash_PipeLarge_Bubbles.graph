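# Sound cue for ambient bubble sounds from a large car wash pipe, played as quiet, attenuated audio in the outdoor ambient sound class.
# It loops a fixed three-second delay followed by a random pick from five equally weighted bubble one-shots, varied through a Modulator at default settings.
# The loop feeds only the first quality-level input, leaving the other three empty.
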SoundCue /Game/Sounds/Fort_Prop_Sounds/Exterior/CarWash/SC_CarWash_PipeLarge_Bubbles
root=QualityLevel_0
nodes [
 {id=Delay_0, children [Modulator_0], properties {delayMax=3, delayMin=3}}
 {id=Looping_0, children [Delay_0]}
 {id=Modulator_0, children [Random_0]}
 {id=QualityLevel_0, children [Looping_0,null,null,null]}
 {id=Random_0, children [WavePlayer_1,WavePlayer_2,WavePlayer_3,WavePlayer_4,WavePlayer_5], properties {weights [1,1,1,1,1]}}
 {id=WavePlayer_1, properties {soundWaveAssetPtr=&/CarWash_LargePipe_BubbleOneShot_01}}
 {id=WavePlayer_2, properties {soundWaveAssetPtr=&/CarWash_LargePipe_BubbleOneShot_02}}
 {id=WavePlayer_3, properties {soundWaveAssetPtr=&/CarWash_LargePipe_BubbleOneShot_03}}
 {id=WavePlayer_4, properties {soundWaveAssetPtr=&/CarWash_LargePipe_BubbleOneShot_04}}
 {id=WavePlayer_5, properties {soundWaveAssetPtr=&/CarWash_LargePipe_BubbleOneShot_05}}
]
properties {
 attenuationSettings=&/ATT_CarWash_Pipe_Loop_01
 maxDistance=1500
 soundClassObject=&Fort_Ambient_Outdoors_Athena
 volumeMultiplier=0.2
}
