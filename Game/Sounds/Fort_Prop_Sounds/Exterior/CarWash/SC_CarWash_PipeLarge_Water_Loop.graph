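# Looping ambient SoundCue for the water sound of a large car wash pipe prop, classed as outdoor Athena ambience.
# A Mixer layers two looping waves: a bubbles track pitched down by a fixed Modulator and a plain hum, with the hum weighted slightly higher.
# The mix plays at reduced overall volume with pipe-loop attenuation limiting its audible range.
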SoundCue /Game/Sounds/Fort_Prop_Sounds/Exterior/CarWash/SC_CarWash_PipeLarge_Water_Loop
root=Mixer_1
nodes [
 {id=Looping_0, children [Modulator_0]}
 {id=Looping_1, children [WavePlayer_2]}
 {id=Mixer_1, children [Looping_0,Looping_1], properties {inputVolume [0.6,0.8]}}
 {id=Modulator_0, children [WavePlayer_0], properties {pitchMax=0.7, pitchMin=0.7, volumeMax=1, volumeMin=1}}
 {id=WavePlayer_0, properties {bLooping=true, soundWaveAssetPtr=&/CarWash_LargePipe_Bubbles}}
 {id=WavePlayer_2, properties {bLooping=true, soundWaveAssetPtr=&/CarWash_LargePipe_Hum}}
]
properties {
 attenuationSettings=&/ATT_CarWash_Pipe_Loop_01
 maxDistance=1500
 soundClassObject=&Fort_Ambient_Outdoors_Athena
 volumeMultiplier=0.3
}
