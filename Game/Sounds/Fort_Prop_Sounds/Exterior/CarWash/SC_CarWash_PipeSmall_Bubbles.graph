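# Sound cue for the small car wash pipe prop, playing a continuous bubbling ambience in the outdoor ambient sound class with pipe-specific attenuation.
# A looping node repeats an equally weighted random pick from five bubble one-shot waves, each passing through a modulator that randomizes pitch upward for variation.
# The quality-level root connects this chain only to its first slot, leaving the other three empty.
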SoundCue /Game/Sounds/Fort_Prop_Sounds/Exterior/CarWash/SC_CarWash_PipeSmall_Bubbles
root=QualityLevel_0
nodes [
 {id=Looping_0, children [Modulator_0]}
 {id=Modulator_0, children [Random_0], properties {pitchMax=1.5, pitchMin=1}}
 {id=QualityLevel_0, children [Looping_0,null,null,null]}
 {id=Random_0, children [WavePlayer_0,WavePlayer_6,WavePlayer_7,WavePlayer_8,WavePlayer_9], properties {weights [1,1,1,1,1]}}
 {id=WavePlayer_0, properties {soundWaveAssetPtr=&/CarWash_SmallPipe_BubbleOneShot_01}}
 {id=WavePlayer_6, properties {soundWaveAssetPtr=&/CarWash_SmallPipe_BubbleOneShot_02}}
 {id=WavePlayer_7, properties {soundWaveAssetPtr=&/CarWash_SmallPipe_BubbleOneShot_03}}
 {id=WavePlayer_8, properties {soundWaveAssetPtr=&/CarWash_SmallPipe_BubbleOneShot_04}}
 {id=WavePlayer_9, properties {soundWaveAssetPtr=&/CarWash_SmallPipe_BubbleOneShot_05}}
]
properties {
 attenuationSettings=&/ATT_CarWash_Pipe_Loop_01
 maxDistance=1500
 soundClassObject=&Fort_Ambient_Outdoors_Athena
 volumeMultiplier=1
}
