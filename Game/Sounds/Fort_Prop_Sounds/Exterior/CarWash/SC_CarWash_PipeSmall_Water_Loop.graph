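# Sound cue for the looping water bubble ambience of a small car wash pipe prop.
# A mixer layers two looping bubble waves, with the first quieter and raised to a fixed higher pitch by a modulator.
# The cue plays at reduced volume and raised overall pitch, using pipe attenuation, an indoor ambient sound class and a concurrency override that caps simultaneous instances.
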SoundCue /Game/Sounds/Fort_Prop_Sounds/Exterior/CarWash/SC_CarWash_PipeSmall_Water_Loop
root=Mixer_1
nodes [
 {id=Looping_0, children [Modulator_0]}
 {id=Looping_1, children [WavePlayer_4]}
 {id=Mixer_1, children [Looping_0,Looping_1], properties {inputVolume [0.6,1]}}
 {id=Modulator_0, children [WavePlayer_3], properties {pitchMax=1.2, pitchMin=1.2, volumeMax=1, volumeMin=1}}
 {
  id=WavePlayer_3
  properties {bLooping=true, soundWaveAssetPtr=&/CarWash_SmallPipe_Bubbles_Loop_01}
 }
 {
  id=WavePlayer_4
  properties {bLooping=true, soundWaveAssetPtr=&/CarWash_SmallPipe_Bubbles_Loop_02}
 }
]
properties {
 attenuationSettings=&/ATT_CarWash_Pipe_Loop_01
 bOverrideConcurrency=true
 maxDistance=1500
 pitchMultiplier=1.2
 soundClassObject=&Fort_Ambient_Indoors_Athena
 volumeMultiplier=0.5
 concurrencyOverrides {maxCount=2}
}
