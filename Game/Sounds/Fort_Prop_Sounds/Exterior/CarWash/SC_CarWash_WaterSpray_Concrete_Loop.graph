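# Sound cue for water spraying on concrete at a car wash, an ambient prop loop built from an interior splashing wave.
# A quality-level root feeds the same wave player to the first two levels and leaves the other two empty, so those play nothing.
# The cue raises pitch, lowers volume, uses close-range car wash attenuation and the Athena indoor ambient class, and caps concurrency by stopping the quietest instance.
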SoundCue /Game/Sounds/Fort_Prop_Sounds/Exterior/CarWash/SC_CarWash_WaterSpray_Concrete_Loop
root=QualityLevel_0
nodes [
 {id=QualityLevel_0, children [WavePlayer_1,WavePlayer_1,null,null]}
 {id=WavePlayer_1, properties {soundWaveAssetPtr=&/Spot_Water_Splashing_Concrete_Int_Loop_02}}
]
properties {
 attenuationSettings=&/ATT_CarWash_WaterSpray_Loop_Close
 bOverrideConcurrency=true
 maxDistance=3300
 pitchMultiplier=1.3
 soundClassObject=&Fort_Ambient_Indoors_Athena
 volumeMultiplier=0.5
 concurrencyOverrides {maxCount=3, resolutionRule=StopQuietest}
}
