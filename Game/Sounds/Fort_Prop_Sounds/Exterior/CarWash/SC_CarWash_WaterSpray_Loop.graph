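# Looping sound cue for car wash water spray, routed to an indoor ambient sound class.
# A root mixer blends a close layer with a quieter distant layer, each shaped by its own attenuation asset.
# The close layer mixes three looping spray waves, and the distant layer mixes two distant variants.
# All waves in a layer play simultaneously under a Looping node, with no random selection.
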SoundCue /Game/Sounds/Fort_Prop_Sounds/Exterior/CarWash/SC_CarWash_WaterSpray_Loop
root=Mixer_1
nodes [
 {
  id=Attenuation_0
  children [Looping_1]
  properties {attenuationSettings=&/ATT_CarWash_WaterSpray_Loop_Close}
 }
 {
  id=Attenuation_1
  children [Looping_2]
  properties {attenuationSettings=&/ATT_CarWash_WaterSpray_Loop_Distant}
 }
 {id=Looping_1, children [Mixer_0]}
 {id=Looping_2, children [Mixer_2]}
 {id=Mixer_0, children [WavePlayer_10, WavePlayer_11, WavePlayer_12], properties {inputVolume [1,1,0.8]}}
 {id=Mixer_1, children [Attenuation_0, Attenuation_1], properties {inputVolume [1,0.5]}}
 {id=Mixer_2, children [WavePlayer_2,WavePlayer_3], properties {inputVolume [1,1]}}
 {id=WavePlayer_10, properties {bLooping=true, soundWaveAssetPtr=&/CarWash_WaterSpray_003}}
 {id=WavePlayer_11, properties {bLooping=true, soundWaveAssetPtr=&/CarWash_WaterSpray_002}}
 {id=WavePlayer_12, properties {bLooping=true, soundWaveAssetPtr=&/CarWash_WaterSpray_001}}
 {id=WavePlayer_2, properties {bLooping=true, soundWaveAssetPtr=&/CarWash_WaterSpray_Distant_01}}
 {id=WavePlayer_3, properties {bLooping=true, soundWaveAssetPtr=&/CarWash_WaterSpray_Distant_02}}
]
properties {maxDistance=4652.9927, soundClassObject=&Fort_Ambient_Indoors_Athena, volumeMultiplier=1.5}
